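; Parent Blueprint for hidden Battle Star quest props that appear when an overlap sphere is entered and shrink away when a second is exited.
; Reveal activates sound and light, scales the star up, then loops a bob-and-spin timeline; hide scales it down and deactivates it.
; Quest initialization binds overlaps and selects a party-assist material for finished shared quests; completion unbinds, then hides or opacity-fades.
; The construction script can swap in a banner mesh and texture.
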(event EventBeginPlay
  (|CallParentFunction)
  (Components|Activation|Deactivate (Variables|Default|GetAmbientSound))
  (Rendering|Components|Light|SetIntensity (Variables|Default|GetAmbientLight) 0.0)
  (Variables|Default|SetSMLocationInit (Variables|/Script/Engine.SceneComponent|GetRelativeLocation (Variables|Default|GetSM_BattlePassStar))))

(event EventQuestInitializedAndValid
  (|CallParentFunction)
  (Collision|BindEventtoOnComponentBeginOverlap (Variables|Default|GetSphereOverlapReveal) (EventDispatchers|CreateEvent :!event Custom|Star%20Reveal))
  (Collision|BindEventtoOnComponentEndOverlap (Variables|Default|GetSphereOverlapHide) (EventDispatchers|CreateEvent :!event Custom|Star%20Hide))
  ;[1 1 1] delay nodes don't delay for the child if they're called in the parents logic. so we fire off of the quest initialized event that fires once we know we've done the checks for these booleans that determine what DMI we're using
  (if (and (Variables|Default|GetQuestAlreadyFinished) (Variables|Default|GetbIsQuestShared))
    (Variables|Default|SetbUseTranslucentStar true)
    (bind _returnvalue (Rendering|Material|CreateDynamicMaterialInstance (Variables|Default|GetSM_BattlePassStar) 0 "/Game/Athena/Items/QuestInteractables/HiddenStars/FX/M_PartyAssistBattleStar.M_PartyAssistBattleStar" "None" :!class /Script/Engine.PrimitiveComponent))
    (Variables|Default|SetBattleStarMat _returnvalue)
    (else
      (bind _returnvalue_2 (Rendering|Material|CreateDynamicMaterialInstance (Variables|Default|GetSM_BattlePassStar) 0 "/Game/Athena/Items/QuestInteractables/HiddenStars/FX/M_BattlePassStar.M_BattlePassStar" "None" :!class /Script/Engine.PrimitiveComponent))
      (Variables|Default|SetBattleStarMat _returnvalue_2))))

(event EventThisQuestObjectiveComplete
  (params (IneractingPawn :type Fort%20Pawn%20Object%20Reference))
  (|CallParentFunction IneractingPawn)
  (Development|PrintString "QuestComplete" true true "(R=0.000000,G=0.660000,B=1.000000,A=1.000000)" 5.0)
  (Collision|UnbindEventfromOnComponentBeginOverlap (Variables|Default|GetSphereOverlapReveal) (EventDispatchers|CreateEvent :!event Custom|Star%20Reveal))
  (Collision|UnbindEventfromOnComponentEndOverlap (Variables|Default|GetSphereOverlapHide) (EventDispatchers|CreateEvent :!event Custom|Star%20Hide))
  (if (Variables|Default|GetbUseTranslucentStar)
    (CallFunction|StarHide 0)
    (else
      (bind (_direction _fadetime) (Timeline|Fade%20Out PlayFromStart)
        (:Update
          (Rendering|Material|SetScalarParameterValue (Variables|Default|GetBattleStarMat) "Opacity" _fadetime :!class /Script/Engine.MaterialInstanceDynamic))
        (:Finished
          (CallFunction|DeactivateHiddenStar))))))

(event Custom|Star%20Reveal
  (params (OverlappedComponent :type Primitive%20Component%20Object%20Reference) (OtherActor :type Actor%20Object%20Reference) (OtherComp :type Primitive%20Component%20Object%20Reference) (OtherBodyIndex :type Integer) (bFromSweep :type Boolean) (SweepResult :type Hit%20Result%20Structure%20%28by%20ref%29))
  (Development|SetHiddeninGame (Variables|Default|GetSM_BattlePassStar) false false)
  (Components|Activation|Activate (Variables|Default|GetAmbientSound) false)
  (Rendering|Components|Light|SetIntensity (Variables|Default|GetAmbientLight) 3e+03)
  ;[1 1 1] Reveal
  (bind (_direction _revealtime) (Timeline|Reveal PlayFromStart)
    (:Update
      (Transformation|SetWorldScale3D (Variables|Default|GetSM_BattlePassStar) (Math|Vector|Lerp(Vector) "0.010000,0.010000,0.010000" "1.000000,1.000000,1.000000" _revealtime)))
    (:Finished
      ;[1 1 1] Make Object Bob and Spin
      (Components|Timeline|SetLooping (Variables|Prop_QuestInteractable_HiddenStar_Parent|GetBob%20and%20Spin) true)
      (bind (_direction_2 _bob _spin) (Timeline|Bob%20and%20Spin Play)
        (:Update
          (Transformation|SetRelativeLocationAndRotation (Variables|Default|GetSM_BattlePassStar) (Math|Vector|Lerp(Vector) (Variables|Default|GetSMLocationInit) (+ (Variables|Default|GetSMLocationInit) "0, 0,20.000000") _bob) (Math|Rotator|Lerp(Rotator) "0, 0, 0" "0,720.000000, 0" _spin) false false))))))

(event Custom|Star%20Hide
  (params (OverlappedComponent :type Primitive%20Component%20Object%20Reference) (OtherActor :type Actor%20Object%20Reference) (OtherComp :type Primitive%20Component%20Object%20Reference) (OtherBodyIndex :type Integer))
  (Utilities|FlowControl|Sequence
    (:then_0
      ;[1 1 1] Reveal
      (Timeline|Reveal Stop))
    (:then_1
      _)
    (:then_2
      (bind (_direction _scaledowntime) (Timeline|Disappear PlayFromStart)
        (:Update
          (Transformation|SetWorldScale3D (Variables|Default|GetSM_BattlePassStar) (Math|Vector|Lerp(Vector) "1.000000,1.000000,1.000000" "0.010000,0.010000,0.010000" _scaledowntime)))
        (:Finished
          (CallFunction|DeactivateHiddenStar)))))
  (Timeline|Bob%20and%20Spin Stop))

(event Custom|DeactivateHiddenStar
  (Development|SetHiddeninGame (Variables|Default|GetSM_BattlePassStar) true false)
  (Components|Activation|Deactivate (Variables|Default|GetAmbientSound))
  (Rendering|Components|Light|SetIntensity (Variables|Default|GetAmbientLight) 0.0))

(fn ConstructionScript
  (bind _sm_battlepassstar (Variables|Default|GetSM_BattlePassStar))
  (|CallParentFunction)
  (if (Variables|Default|GetisBanner)
    (Transformation|SetRelativeLocation _sm_battlepassstar "0, 0,30.000000" false false)
    (Components|StaticMesh|SetStaticMesh _sm_battlepassstar "/Game/Athena/Items/QuestInteractables/Cumulative/FX/SM_Athena_Cumulative_Banner.SM_Athena_Cumulative_Banner")
    (bind _returnvalue (Rendering|Material|CreateDynamicMaterialInstance _sm_battlepassstar 1 "/Game/Athena/Items/QuestInteractables/Cumulative/FX/M_GoldBanner_Symbol_TomatoHead.M_GoldBanner_Symbol_TomatoHead" "None" :!class /Script/Engine.PrimitiveComponent))
    (Rendering|Material|SetTextureParameterValue _returnvalue "BannerTexture" (Variables|Default|GetBannerTexture))))
